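# Physics asset for the LocalZilla male medium body, covering its dynamic jacket and collar bones alongside torso, neck, upper-arm and thigh bodies.
# Two-segment front and side jacket chains hang from small master bodies on spine_03, with cross-constraints tying front to side on each side.
# Front and mid collar boxes hang from masters on spine_05, while lower collar capsules attach there directly with tighter angular limits.
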PhysicsAsset /BRCosmetics/Characters/Player/Male/Medium/Bodies/M_MED_LocalZilla/Meshes/M_MED_LocalZilla_Physics
bodies {
 clavicle_l [{shapeType=Capsule, length=5, radius=9, shapeName=clavicle_l_capsule, center=<10.3884 0 2.0219>, rotation=<rot 67.1501 -0.0001 -0.0001>}]
 clavicle_r [{shapeType=Capsule, length=5, radius=9, shapeName=clavicle_l_capsule, center=<-10.3884 0 -2.0219>, rotation=<rot 67.1501 -0.0001 -0.0001>}]
 dyn_L_frontCollarLower [{shapeType=Capsule, length=5, radius=4, shapeName=dyn_L_frontCollarLower_capsule, center=<-1.6258 0 0>}]
 dyn_L_frontCollar_1 [{shapeType=Box, extentX=15, extentY=1.899, extentZ=7.909, shapeName=dyn_L_frontCollar_1_box, center=<5 0 0>, rotation=<rot 0 0 52>}]
 dyn_L_frontCollar_master [{shapeType=Capsule, length=0.5, radius=0.5, shapeName=dyn_L_frontCollar_master_capsule}]
 dyn_L_frontJacket_1 [{shapeType=Capsule, length=8, radius=7, shapeName=dyn_L_frontJacket_1_capsule, center=<3.5 0 -7>, rotation=<rot 90 0 0>}]
 dyn_L_frontJacket_1_master [{shapeType=Capsule, length=0.5, radius=0.5, shapeName=dyn_L_frontJacket_1_master_capsule}]
 dyn_L_frontJacket_2 [{shapeType=Capsule, length=8, radius=7, shapeName=dyn_L_frontJacket_2_capsule, center=<3.5 0 -6.5>, rotation=<rot 90 0 0>}]
 dyn_L_midCollar_1 [{shapeType=Box, extentX=15, extentY=1.899, extentZ=7.909, shapeName=dyn_L_midCollar_1_box, center=<2 0 0>}]
 dyn_L_midCollar_master [{shapeType=Capsule, length=0.5, radius=0.5, shapeName=dyn_L_midCollar_master_capsule}]
 dyn_L_sideJacket_1 [{shapeType=Capsule, length=8, radius=7, shapeName=dyn_L_sideJacket_1_capsule, center=<3.5 0 -7.5>, rotation=<rot 90 0 0>}]
 dyn_L_sideJacket_1_master [{shapeType=Capsule, length=0.5, radius=0.5, shapeName=dyn_L_sideJacket_1_master_capsule}]
 dyn_L_sideJacket_2 [{shapeType=Capsule, length=8, radius=7, shapeName=dyn_L_sideJacket_2_capsule, center=<3.5 0 -7.5>, rotation=<rot 90 0 0>}]
 dyn_R_frontCollarLower [{shapeType=Capsule, length=5, radius=4, shapeName=dyn_R_frontCollarLower_capsule, center=<1.6258 0 0>}]
 dyn_R_frontCollar_1 [{shapeType=Box, extentX=15, extentY=1.899, extentZ=7.909, shapeName=dyn_R_frontCollar_1_box, center=<-5 0 0>, rotation=<rot 0 0 52>}]
 dyn_R_frontCollar_master [{shapeType=Capsule, length=0.5, radius=0.5, shapeName=dyn_R_frontCollar_master_capsule}]
 dyn_R_frontJacket_1 [{shapeType=Capsule, length=8, radius=7, shapeName=dyn_R_frontJacket_1_capsule, center=<-3.5 0 7>, rotation=<rot 90 0 0>}]
 dyn_R_frontJacket_1_master [{shapeType=Capsule, length=0.5, radius=0.5, shapeName=dyn_R_frontJacket_1_master_capsule}]
 dyn_R_frontJacket_2 [{shapeType=Capsule, length=8, radius=7, shapeName=dyn_R_frontJacket_2_capsule, center=<-3.5 0 6.5>, rotation=<rot 90 0 0>}]
 dyn_R_midCollar_1 [{shapeType=Box, extentX=15, extentY=1.899, extentZ=7.909, shapeName=dyn_R_midCollar_1_box, center=<-2 0 0>}]
 dyn_R_midCollar_master [{shapeType=Capsule, length=0.5, radius=0.5, shapeName=dyn_R_midCollar_master_capsule}]
 dyn_R_sideJacket_1 [{shapeType=Capsule, length=8, radius=7, shapeName=dyn_R_sideJacket_1_capsule, center=<-3.5 0 7.5>, rotation=<rot 90 0 0>}]
 dyn_R_sideJacket_1_master [{shapeType=Capsule, length=0.5, radius=0.5, shapeName=dyn_R_sideJacket_1_master_capsule}]
 dyn_R_sideJacket_2 [{shapeType=Capsule, length=8, radius=7, shapeName=dyn_R_sideJacket_2_capsule, center=<-3.5 0 7.5>, rotation=<rot 90 0 0>}]
 neck_01 [{shapeType=Capsule, length=5, radius=8, shapeName=neck_02_capsule, center=<2.2154 0.8029 0>, rotation=<rot 90 -15.9454 -5.9453>}]
 pelvis [{shapeType=Capsule, length=7.4309, radius=12.5, shapeName=pelvis_capsule}]
 spine_01 [{shapeType=Capsule, length=7.4309, radius=11.5, shapeName=spine_01_capsule}]
 spine_02 [{shapeType=Capsule, length=7.4309, radius=11.5, shapeName=spine_02_capsule}]
 spine_03 [{shapeType=Capsule, length=7.4309, radius=11.5, shapeName=spine_03_capsule}]
 spine_05 [{shapeType=Capsule, length=7.4309, radius=15, shapeName=spine_03_capsule}]
 thigh_l [{shapeType=Capsule, length=30, radius=9, shapeName=thigh_l_capsule, center=<-15.2741 0 0>, rotation=<rot 90 0 0>}]
 thigh_r [{shapeType=Capsule, length=30, radius=9, shapeName=thigh_r_capsule, center=<15.2741 0 0>, rotation=<rot 90 0 0>}]
 upperarm_l [{shapeType=Capsule, length=10, radius=9, shapeName=upperarm_l_capsule, center=<14 0 0>, rotation=<rot 85 0 0>}]
 upperarm_r [{shapeType=Capsule, length=10, radius=9, shapeName=upperarm_l_capsule, center=<-14 0 0>, rotation=<rot 85 0 0>}]
}
constraints [
 {
  bone1Name=spine_01
  bone2Name=pelvis
  swing1LimitDegrees=45
  swing1Motion=Free
  swing2LimitDegrees=45
  swing2Motion=Free
  twistLimitDegrees=45
  twistMotion=Free
 }
 {
  bone1Name=spine_02
  bone2Name=spine_01
  swing1LimitDegrees=45
  swing1Motion=Free
  swing2LimitDegrees=45
  swing2Motion=Free
  twistLimitDegrees=45
  twistMotion=Free
 }
 {
  bone1Name=spine_03
  bone2Name=spine_02
  swing1LimitDegrees=45
  swing1Motion=Free
  swing2LimitDegrees=45
  swing2Motion=Free
  twistLimitDegrees=45
  twistMotion=Free
 }
 {
  bone1Name=dyn_L_frontJacket_1
  bone2Name=dyn_L_frontJacket_1_master
  swing1LimitDegrees=60
  swing2LimitDegrees=45
  swing2Motion=Free
  twistLimitDegrees=45
  twistMotion=Free
 }
 {
  bone1Name=dyn_L_frontJacket_2
  bone2Name=dyn_L_frontJacket_1
  swing1LimitDegrees=60
  swing2LimitDegrees=45
  swing2Motion=Free
  twistLimitDegrees=45
  twistMotion=Free
 }
 {
  bone1Name=dyn_L_sideJacket_1
  bone2Name=dyn_L_sideJacket_1_master
  swing1LimitDegrees=45
  swing2LimitDegrees=45
  swing2Motion=Free
  twistLimitDegrees=45
  twistMotion=Free
 }
 {
  bone1Name=dyn_L_sideJacket_2
  bone2Name=dyn_L_sideJacket_1
  swing1LimitDegrees=45
  swing2LimitDegrees=45
  swing2Motion=Free
  twistLimitDegrees=45
  twistMotion=Free
 }
 {
  bone1Name=dyn_R_frontJacket_1
  bone2Name=dyn_R_frontJacket_1_master
  swing1LimitDegrees=60
  swing2LimitDegrees=45
  swing2Motion=Free
  twistLimitDegrees=45
  twistMotion=Free
 }
 {
  bone1Name=dyn_R_frontJacket_2
  bone2Name=dyn_R_frontJacket_1
  swing1LimitDegrees=60
  swing2LimitDegrees=45
  swing2Motion=Free
  twistLimitDegrees=45
  twistMotion=Free
 }
 {
  bone1Name=dyn_R_sideJacket_1
  bone2Name=dyn_R_sideJacket_1_master
  swing1LimitDegrees=45
  swing2LimitDegrees=45
  swing2Motion=Free
  twistLimitDegrees=45
  twistMotion=Free
 }
 {
  bone1Name=dyn_R_sideJacket_2
  bone2Name=dyn_R_sideJacket_1
  swing1LimitDegrees=45
  swing2LimitDegrees=45
  swing2Motion=Free
  twistLimitDegrees=45
  twistMotion=Free
 }
 {bone1Name=dyn_L_frontCollar_1, bone2Name=dyn_L_frontCollar_master, swing1LimitDegrees=45, swing2LimitDegrees=45, twistLimitDegrees=45}
 {bone1Name=dyn_L_midCollar_1, bone2Name=dyn_L_midCollar_master, swing1LimitDegrees=45, swing2LimitDegrees=45, twistLimitDegrees=45}
 {bone1Name=dyn_R_frontCollar_1, bone2Name=dyn_R_frontCollar_master, swing1LimitDegrees=45, swing2LimitDegrees=45, twistLimitDegrees=45}
 {bone1Name=dyn_R_midCollar_1, bone2Name=dyn_R_midCollar_master, swing1LimitDegrees=45, swing2LimitDegrees=45, twistLimitDegrees=45}
 {
  bone1Name=clavicle_l
  bone2Name=spine_05
  swing1LimitDegrees=45
  swing1Motion=Free
  swing2LimitDegrees=45
  swing2Motion=Free
  twistLimitDegrees=45
  twistMotion=Free
 }
 {
  bone1Name=clavicle_r
  bone2Name=spine_05
  swing1LimitDegrees=45
  swing1Motion=Free
  swing2LimitDegrees=45
  swing2Motion=Free
  twistLimitDegrees=45
  twistMotion=Free
 }
 {
  bone1Name=spine_05
  bone2Name=spine_03
  swing1LimitDegrees=45
  swing1Motion=Free
  swing2LimitDegrees=45
  swing2Motion=Free
  twistLimitDegrees=45
  twistMotion=Free
 }
 {
  bone1Name=neck_01
  bone2Name=spine_05
  swing1LimitDegrees=45
  swing1Motion=Free
  swing2LimitDegrees=45
  swing2Motion=Free
  twistLimitDegrees=45
  twistMotion=Free
 }
 {
  bone1Name=thigh_l
  bone2Name=pelvis
  swing1LimitDegrees=45
  swing1Motion=Free
  swing2LimitDegrees=45
  swing2Motion=Free
  twistLimitDegrees=45
  twistMotion=Free
 }
 {
  bone1Name=thigh_r
  bone2Name=pelvis
  swing1LimitDegrees=45
  swing1Motion=Free
  swing2LimitDegrees=45
  swing2Motion=Free
  twistLimitDegrees=45
  twistMotion=Free
 }
 {
  bone1Name=dyn_L_frontJacket_2
  bone2Name=dyn_L_sideJacket_2
  swing1LimitDegrees=45
  swing1Motion=Free
  swing2LimitDegrees=45
  swing2Motion=Free
  twistLimitDegrees=45
  twistMotion=Free
 }
 {
  bone1Name=dyn_L_frontJacket_1
  bone2Name=dyn_L_sideJacket_1
  swing1LimitDegrees=45
  swing1Motion=Free
  swing2LimitDegrees=45
  swing2Motion=Free
  twistLimitDegrees=45
  twistMotion=Free
 }
 {
  bone1Name=upperarm_l
  bone2Name=clavicle_l
  swing1LimitDegrees=45
  swing1Motion=Free
  swing2LimitDegrees=45
  swing2Motion=Free
  twistLimitDegrees=45
  twistMotion=Free
 }
 {
  bone1Name=upperarm_r
  bone2Name=clavicle_r
  swing1LimitDegrees=45
  swing1Motion=Free
  swing2LimitDegrees=45
  swing2Motion=Free
  twistLimitDegrees=45
  twistMotion=Free
 }
 {
  bone1Name=dyn_L_frontJacket_1_master
  bone2Name=spine_03
  swing1LimitDegrees=45
  swing1Motion=Free
  swing2LimitDegrees=45
  swing2Motion=Free
  twistLimitDegrees=45
  twistMotion=Free
 }
 {
  bone1Name=dyn_L_sideJacket_1_master
  bone2Name=spine_03
  swing1LimitDegrees=45
  swing1Motion=Free
  swing2LimitDegrees=45
  swing2Motion=Free
  twistLimitDegrees=45
  twistMotion=Free
 }
 {
  bone1Name=dyn_R_frontJacket_1_master
  bone2Name=spine_03
  swing1LimitDegrees=45
  swing1Motion=Free
  swing2LimitDegrees=45
  swing2Motion=Free
  twistLimitDegrees=45
  twistMotion=Free
 }
 {
  bone1Name=dyn_R_sideJacket_1_master
  bone2Name=spine_03
  swing1LimitDegrees=45
  swing1Motion=Free
  swing2LimitDegrees=45
  swing2Motion=Free
  twistLimitDegrees=45
  twistMotion=Free
 }
 {
  bone1Name=dyn_R_frontJacket_1
  bone2Name=dyn_R_sideJacket_1
  swing1LimitDegrees=45
  swing1Motion=Free
  swing2LimitDegrees=45
  swing2Motion=Free
  twistLimitDegrees=45
  twistMotion=Free
 }
 {
  bone1Name=dyn_R_frontJacket_2
  bone2Name=dyn_R_sideJacket_2
  swing1LimitDegrees=45
  swing1Motion=Free
  swing2LimitDegrees=45
  swing2Motion=Free
  twistLimitDegrees=45
  twistMotion=Free
 }
 {
  bone1Name=dyn_L_frontCollar_master
  bone2Name=spine_05
  swing1LimitDegrees=45
  swing1Motion=Free
  swing2LimitDegrees=45
  swing2Motion=Free
  twistLimitDegrees=45
  twistMotion=Free
 }
 {
  bone1Name=dyn_R_frontCollar_master
  bone2Name=spine_05
  swing1LimitDegrees=45
  swing1Motion=Free
  swing2LimitDegrees=45
  swing2Motion=Free
  twistLimitDegrees=45
  twistMotion=Free
 }
 {
  bone1Name=dyn_L_midCollar_master
  bone2Name=spine_05
  swing1LimitDegrees=45
  swing1Motion=Free
  swing2LimitDegrees=45
  swing2Motion=Free
  twistLimitDegrees=45
  twistMotion=Free
 }
 {
  bone1Name=dyn_R_midCollar_master
  bone2Name=spine_05
  swing1LimitDegrees=45
  swing1Motion=Free
  swing2LimitDegrees=45
  swing2Motion=Free
  twistLimitDegrees=45
  twistMotion=Free
 }
 {bone1Name=dyn_L_frontCollarLower, bone2Name=spine_05, swing1LimitDegrees=45, swing2LimitDegrees=0, twistLimitDegrees=5}
 {bone1Name=dyn_R_frontCollarLower, bone2Name=spine_05, swing1LimitDegrees=45, swing2LimitDegrees=0, twistLimitDegrees=5}
]
