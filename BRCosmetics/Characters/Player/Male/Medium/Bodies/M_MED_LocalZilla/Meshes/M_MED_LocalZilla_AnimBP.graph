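# Post-process-style AnimBlueprint for the LocalZilla male medium body cosmetic, adding secondary motion to jacket, collar, necklaces and zipper.
# The copied parent pose runs through a Control Rig, constraints, AnimDynamics and RigidBody physics, with corrective bone offsets driven by limb angles.
# A front-end idle curve blends slot animation onto collar bones.
# The event graph also sets physics gravity from movement state and feeds velocity and skydive wind parameters to WPO cloth materials.
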AnimBlueprint /BRCosmetics/Characters/Player/Male/Medium/Bodies/M_MED_LocalZilla/Meshes/M_MED_LocalZilla_AnimBP
target_skeleton=&/M_MED_LocalZilla_Skeleton
graphs [
 AnimGraph {nodes [
   n1 = Layered blend per bone(
    BasePose=Use cached pose 'startPose'()
    BlendPoses_0=Slot 'DefaultSlot'()
    BlendWeights_0=Get customIdleIsPlayingFloat() -> (customIdleIsPlayingFloat:float)
   ) {layerSetup [{branchFilters [
       {boneName=dyn_L_frontCollar_master}
       {boneName=dyn_L_frontCollar_1}
       {boneName=dyn_R_frontCollar_master}
       {boneName=dyn_R_frontCollar_1}
       {boneName=dyn_L_midCollar_master}
       {boneName=dyn_L_midCollar_1}
       {boneName=dyn_R_midCollar_master}
       {boneName=dyn_R_midCollar_1}
      ]}]}
   Output Pose(Result=@n1) {layerGroup=None}
   n2 = Get R upperarm alpha() -> (R upperarm alpha:double)
   n3 = Control Rig(Source=Copy Pose From Mesh() {bUseAttachedParent=true}) {
    alphaBoolBlend {blendOption=Linear}
    controlRigAssetReference {blueprintRigClass=&/M_MED_LocalZilla_DeformRig_CtrlRig.M_MED_LocalZilla_DeformRig_CtrlRig_C}
   }
   n4 = "Transform (Modify) Bone - Bone: attach_backpack"(
    ComponentPose=Local To Component(LocalPose=@n3)
    Alpha=Get armsBack() -> (armsBack:double)
    Translation="0,-2.5,0"
    Rotation="0,0,0"
    Scale="1,1,1"
   ) {
    translationMode=BMM_Additive
    translationSpace=BCS_BoneSpace
    alphaBoolBlend {blendOption=Linear}
    alphaScaleBiasClamp {bInterpResult=true, interpSpeedDecreasing=20, interpSpeedIncreasing=20}
    boneToModify {boneName=attach_backpack}
    translation {y=-2.5}
   }
   n5 = "Constraint - dyn_placeholderNeck_01 (P)"(ComponentPose=@n4, Alpha=1, ConstraintWeights_0=1) {
    alphaBoolBlend {blendOption=Linear}
    boneToModify {boneName=dyn_placeholderNeck_01}
    constraintSetup [{targetBone {boneName=neck_01}}]
    constraintWeights [1]
   }
   n6 = "Constraint - dyn_placeholderNeck_02 (P)"(ComponentPose=@n5, Alpha=1, ConstraintWeights_0=1) {
    alphaBoolBlend {blendOption=Linear}
    boneToModify {boneName=dyn_placeholderNeck_02}
    constraintSetup [{targetBone {boneName=neck_02}}]
    constraintWeights [1]
   }
   n7 = "Constraint - dyn_placeholderHead (P)"(ComponentPose=@n6, Alpha=1, ConstraintWeights_0=1) {
    alphaBoolBlend {blendOption=Linear}
    boneToModify {boneName=dyn_placeholderHead}
    constraintSetup [{targetBone {boneName=head}}]
    constraintWeights [1]
   }
   n8 = Anim Dynamics - Bone: dyn_topNecklace(ComponentPose=@n7, Alpha=1, ExternalForce="0,0,0") {
    angularDampingOverride=1
    angularSpringConstant=20000
    bAngularSpring=true
    bOverrideAngularDamping=true
    bUseGravityOverride=true
    gravityScale=0.35
    simulationSpace=RootRelative
    alphaBoolBlend {blendOption=Linear}
    boundBone {boneName=dyn_topNecklace}
    componentLinearAccScale=<1 1 1>
    componentLinearVelScale=<1 1 1>
    gravityOverride {z=-980}
    physicsBodyDefinitions [{
      boundBone {boneName=dyn_topNecklace}
      boxExtents=<8 8 8>
      constraintSetup {angularLimitsMax=<0 0 7>, angularLimitsMin=<0 -30 -7>, angularTarget=<0 0 0>}
      localJointOffset=<1 0 0>
     }]
   }
   n9 = Anim Dynamics - Bone: dyn_bottomNecklace(ComponentPose=@n8, Alpha=1, ExternalForce="0,0,0") {
    angularDampingOverride=1
    angularSpringConstant=20000
    bAngularSpring=true
    bOverrideAngularDamping=true
    bUseGravityOverride=true
    gravityScale=0.35
    simulationSpace=RootRelative
    alphaBoolBlend {blendOption=Linear}
    boundBone {boneName=dyn_bottomNecklace}
    componentLinearAccScale=<1 1 1>
    componentLinearVelScale=<1 1 1>
    gravityOverride {z=-980}
    physicsBodyDefinitions [{
      boundBone {boneName=dyn_bottomNecklace}
      boxExtents=<13 13 13>
      constraintSetup {angularLimitsMax=<0 0 7>, angularLimitsMin=<0 -30 -7>, angularTarget=<0 0 0>}
      localJointOffset=<1 0 0>
     }]
   }
   n10 = "Constraint - dyn_medal (P)"(ComponentPose=@n9, Alpha=1, ConstraintWeights_0=1) {
    alphaBoolBlend {blendOption=Linear}
    boneToModify {boneName=dyn_medal}
    constraintSetup [{targetBone {boneName=dyn_topNecklace}}]
    constraintWeights [1]
   }
   n11 = Anim Dynamics - Bone: dyn_zipper(ComponentPose=@n10, Alpha=1) {
    angularDampingOverride=1
    angularSpringConstant=200
    bAngularSpring=true
    bOverrideAngularDamping=true
    bUseGravityOverride=true
    gravityScale=0.35
    simulationSpace=RootRelative
    alphaBoolBlend {blendOption=Linear}
    boundBone {boneName=dyn_zipper}
    componentLinearAccScale=<1 1 1>
    componentLinearVelScale=<1 1 1>
    gravityOverride {z=-980}
    physicsBodyDefinitions [{
      boundBone {boneName=dyn_zipper}
      boxExtents=<4 4 4>
      constraintSetup {angularLimitsMax=<3 10 40>, angularLimitsMin=<-3 -10 -20>, angularTarget=<1 0 0>}
      localJointOffset=<3 0 0>
     }]
   }
   n12 = "Transform (Modify) Bone - Bone: dyn_L_frontJacket_1_master"(ComponentPose=@n11, Alpha=Get L thigh up() -> (L thigh up:double), Translation="0,0,0", Rotation="-20,80,0", Scale="1,1,1") {
    rotationMode=BMM_Additive
    rotationSpace=BCS_BoneSpace
    translationMode=BMM_Additive
    translationSpace=BCS_BoneSpace
    alphaBoolBlend {blendOption=Linear}
    boneToModify {boneName=dyn_L_frontJacket_1_master}
    rotation {pitch=-20, yaw=80}
   }
   n13 = "Transform (Modify) Bone - Bone: dyn_R_frontJacket_1_master"(ComponentPose=@n12, Alpha=Get R thigh up() -> (R thigh up:double), Translation="0,0,0", Rotation="-20,80,0", Scale="1,1,1") {
    rotationMode=BMM_Additive
    rotationSpace=BCS_BoneSpace
    translationMode=BMM_Additive
    translationSpace=BCS_BoneSpace
    alphaBoolBlend {blendOption=Linear}
    alphaScaleBiasClamp {scale=0.4}
    boneToModify {boneName=dyn_R_frontJacket_1_master}
    rotation {pitch=-20, yaw=80}
   }
   n14 = Get L upperarm alpha() -> (L upperarm alpha:double)
   n15 = "Transform (Modify) Bone - Bone: dyn_L_frontCollar_master"(ComponentPose=@n13, Alpha=@n14, Translation="0,0,0", Rotation="-5,50,-65", Scale="1,1,1") {
    rotationMode=BMM_Additive
    rotationSpace=BCS_BoneSpace
    translationMode=BMM_Additive
    translationSpace=BCS_BoneSpace
    alphaBoolBlend {blendOption=Linear}
    alphaScaleBiasClamp {bInterpResult=true, interpSpeedDecreasing=20, interpSpeedIncreasing=20}
    boneToModify {boneName=dyn_L_frontCollar_master}
    rotation=<rot -5 50 -65>
   }
   n16 = "Transform (Modify) Bone - Bone: dyn_L_midCollar_master"(ComponentPose=@n15, Alpha=@n14, Translation="0,0,0", Rotation="-5,50,-65", Scale="1,1,1") {
    rotationMode=BMM_Additive
    rotationSpace=BCS_BoneSpace
    translationMode=BMM_Additive
    translationSpace=BCS_BoneSpace
    alphaBoolBlend {blendOption=Linear}
    alphaScaleBiasClamp {bInterpResult=true, interpSpeedDecreasing=20, interpSpeedIncreasing=20, scale=0.5}
    boneToModify {boneName=dyn_L_midCollar_master}
    rotation=<rot -5 50 -65>
   }
   n17 = "Transform (Modify) Bone - Bone: dyn_R_frontCollar_master"(ComponentPose=@n16, Alpha=@n2, Translation="0,0,0", Rotation="-5,50,-65", Scale="1,1,1") {
    rotationMode=BMM_Additive
    rotationSpace=BCS_BoneSpace
    translationMode=BMM_Additive
    translationSpace=BCS_BoneSpace
    alphaBoolBlend {blendOption=Linear}
    alphaScaleBiasClamp {bInterpResult=true, interpSpeedDecreasing=20, interpSpeedIncreasing=20}
    boneToModify {boneName=dyn_R_frontCollar_master}
    rotation=<rot -5 50 -65>
   }
   n18 = "Transform (Modify) Bone - Bone: dyn_R_midCollar_master"(ComponentPose=@n17, Alpha=@n2, Translation="0,0,0", Rotation="-5,50,-65", Scale="1,1,1") {
    rotationMode=BMM_Additive
    rotationSpace=BCS_BoneSpace
    translationMode=BMM_Additive
    translationSpace=BCS_BoneSpace
    alphaBoolBlend {blendOption=Linear}
    alphaScaleBiasClamp {bInterpResult=true, interpSpeedDecreasing=20, interpSpeedIncreasing=20, scale=0.5}
    boneToModify {boneName=dyn_R_midCollar_master}
    rotation=<rot -5 50 -65>
   }
   n19 = RigidBody!(ComponentPose=@n18, Alpha, OverrideWorldGravity:Vector=Get GravityOverride(), ExternalForce) -> (Pose) {
    bClampLinearTranslationLimitToRefPose=true
    bOverrideWorldGravity=true
    overridePhysicsAsset=&/M_MED_LocalZilla_Physics
    simulationSpace=BaseBoneSpace
    alphaBoolBlend {blendOption=Linear}
    baseBoneRef {boneName=spine_03}
    componentLinearAccScale=<0.2 0.2 0.2>
    componentLinearVelScale=<0.22 0.22 0.22>
    overrideWorldGravity {z=-980}
   }
   n20 = Get L upperarm alpha()
   n21 = "Transform (Modify) Bone - Bone: dyn_L_frontCollar_master"(ComponentPose=@n19, Alpha=@n20, Translation="0,-3,6", Rotation="-15,30,-50", Scale="1,1,1") {
    rotationMode=BMM_Additive
    rotationSpace=BCS_BoneSpace
    translationMode=BMM_Additive
    translationSpace=BCS_BoneSpace
    alphaBoolBlend {blendOption=Linear}
    alphaScaleBiasClamp {scale=0.6}
    boneToModify {boneName=dyn_L_frontCollar_master}
    rotation=<rot -15 30 -50>
    translation {y=-3, z=6}
   }
   n22 = "Transform (Modify) Bone - Bone: dyn_L_frontCollarLower_master"(ComponentPose=@n21, Alpha=@n20, Translation="0,0,0", Rotation="-30,-90,0", Scale="1,1,1") {
    rotationMode=BMM_Additive
    rotationSpace=BCS_BoneSpace
    translationMode=BMM_Additive
    translationSpace=BCS_BoneSpace
    alphaBoolBlend {blendOption=Linear}
    boneToModify {boneName=dyn_L_frontCollarLower_master}
    rotation {pitch=-30, yaw=-90}
   }
   n23 = "Transform (Modify) Bone - Bone: dyn_L_midCollar_2"(ComponentPose=@n22, Alpha=@n20, Translation="0,0.5,0", Rotation="0,30,0", Scale="1,1,1") {
    rotationMode=BMM_Additive
    rotationSpace=BCS_BoneSpace
    translationMode=BMM_Additive
    translationSpace=BCS_BoneSpace
    alphaBoolBlend {blendOption=Linear}
    boneToModify {boneName=dyn_L_midCollar_2}
    rotation {yaw=30}
    translation {y=0.5}
   }
   n24 = Get R upperarm alpha()
   n25 = "Transform (Modify) Bone - Bone: dyn_R_frontCollar_master"(ComponentPose=@n23, Alpha=@n24, Translation="0,-3,3", Rotation="-15,30,-50", Scale="1,1,1") {
    rotationMode=BMM_Additive
    rotationSpace=BCS_BoneSpace
    translationMode=BMM_Additive
    translationSpace=BCS_BoneSpace
    alphaBoolBlend {blendOption=Linear}
    alphaScaleBiasClamp {scale=0.6}
    boneToModify {boneName=dyn_R_frontCollar_master}
    rotation=<rot -15 30 -50>
    translation {y=-3, z=3}
   }
   n26 = "Transform (Modify) Bone - Bone: dyn_R_frontCollarLower_master"(ComponentPose=@n25, Alpha=@n24, Translation="0,0,0", Rotation="-30,-90,0", Scale="1,1,1") {
    rotationMode=BMM_Additive
    rotationSpace=BCS_BoneSpace
    translationMode=BMM_Additive
    translationSpace=BCS_BoneSpace
    alphaBoolBlend {blendOption=Linear}
    boneToModify {boneName=dyn_R_frontCollarLower_master}
    rotation {pitch=-30, yaw=-90}
   }
   n27 = "Transform (Modify) Bone - Bone: dyn_R_midCollar_2"(ComponentPose=@n26, Alpha=@n24, Translation="0,-0.5,0", Rotation="0,30,0", Scale="1,1,1") {
    rotationMode=BMM_Additive
    rotationSpace=BCS_BoneSpace
    translationMode=BMM_Additive
    translationSpace=BCS_BoneSpace
    alphaBoolBlend {blendOption=Linear}
    boneToModify {boneName=dyn_R_midCollar_2}
    rotation {yaw=30}
    translation {y=-0.5}
   }
   Save cached pose 'startPose'(Pose=Component To Local(ComponentPose=@n27)) {cacheName=startPose}
   Get neckRotateBack() -> (neckRotateBack:double)
  ]}
 Calculate Cloth Sim Noise {nodes [
   n1 = Calculate Cloth Sim Noise!() -> (
    then
    MID:Material Instance Dynamic Object Reference
    OverlayMID:Material Instance Dynamic Object Reference
    SkelMesh:Skeletal Mesh Component Object Reference
    BoneName:Name=root
    DeltaTime:double=~
    NoiseMovementStrength:double=0.25
    NoiseBoneLocation:"Vector (by ref)"
    NoiseVelocityVector:"Vector (by ref)"
   ) {is=title K2Node_FunctionEntry}
   > n2 = calculateMotionVectorForSocket!(
    execute
    self
    skelMesh:Scene Component Object Reference=@n1.SkelMesh
    boneName:Name=@n1.BoneName
    previousBoneLocation:Vector=@n1.NoiseBoneLocation
   ) -> (then, boneMotionVector:Vector, boneLocation:Vector, boneTransform:Transform)
   n3 = vector + vector(
    A=vector * vector(A=@n2, B=float * float(A=@n1.DeltaTime, B=@n1.NoiseMovementStrength) {is=title K2Node_PromotableOperator}) -> (ReturnValue=~)
    B=@n1.NoiseVelocityVector
   )
   > SetVectorParameterValue(
    self=[@n1.OverlayMID, @n1]
    ParameterName=ClothSim_ABP_TurbulenceUVOffset
    Value="To LinearColor (Vector)"!(InVec:Vector=@n3) -> (ReturnValue:Linear Color Structure)
   )
   > "Set Vector (by ref)"!(execute:Exec, Target:"Vector (by ref)"=@n1.NoiseBoneLocation, Value:Vector=@n2.boneLocation) -> (then:Exec) {is=title K2Node_VariableSetRef}
   > "Set Vector (by ref)"(Target=@n1.NoiseVelocityVector, Value=@n3)
   > Return Node()
  ]}
 EventGraph {nodes [
   n1 = Event BlueprintUpdateAnimation()
   > n2 = Sequence!(execute) -> (then_0, then_1, then_2:Exec, then_3:Exec, then_4:Exec, then_5:Exec, then_6:Exec)
   n3 = GetOwnerSkeletalMeshComponent()
   n4 = BreakTransform(InTransform=GetDeltaTransformFromRefPose(self=@n3, BoneName=neck_01, BaseName=spine_05))
   > Set neckRotateBack(neckRotateBack:double=MapRangeClamped(Value=BreakRotator(InRot=@n4.Rotation).Yaw, InRangeA=5, InRangeB=-10)) -> (Output_Get:double)
   n5 = BreakTransform(InTransform=GetDeltaTransformFromRefPose(self=@n3, BoneName=clavicle_l, BaseName=spine_05))
   > Set L clav alpha(L clav alpha:double=MapRangeClamped(Value=BreakRotator(InRot=@n5.Rotation).Pitch, InRangeA=0, InRangeB=30)) -> (Output_Get:double)
   n6 = BreakTransform(InTransform=GetDeltaTransformFromRefPose(self=@n3, BoneName=clavicle_r, BaseName=spine_05))
   > Set R clav alpha(R clav alpha:double=MapRangeClamped(Value=BreakRotator(InRot=@n6.Rotation).Pitch, InRangeA=0, InRangeB=30)) -> (Output_Get:double)
   n7 = BreakRotator(InRot=BreakTransform(InTransform=GetDeltaTransformFromRefPose(self=@n3, BoneName=upperarm_l, BaseName=clavicle_l)).Rotation)
   > Set L upperarm alpha(L upperarm alpha:double="Max (Float)"(A=MapRangeClamped(Value=@n7.Yaw, InRangeA=60, InRangeB=80), B=MapRangeClamped(Value=@n7.Yaw, InRangeA=30, InRangeB=80))) -> (Output_Get:double)
   n8 = BreakRotator(InRot=BreakTransform(InTransform=GetDeltaTransformFromRefPose(self=@n3, BoneName=upperarm_r, BaseName=clavicle_r)).Rotation)
   > Set R upperarm alpha(R upperarm alpha:double="Max (Float)"(A=MapRangeClamped(Value=@n8.Yaw, InRangeA=60, InRangeB=80), B=MapRangeClamped(Value=@n8.Yaw, InRangeA=30, InRangeB=80))) -> (Output_Get:double)
   n9 = Get bIsSkydiveFloating()
   > Set armsBack(armsBack:double="Max (Float)"!(
     A=MapRangeClamped(Value=@n7.Yaw, InRangeA=0, InRangeB=-20)
     B=MapRangeClamped(Value=@n8.Yaw, InRangeA=0, InRangeB=-20)
     C:double="To Float (Boolean)"(InBool=@n9)
    ) -> (ReturnValue)) -> (Output_Get:double)
   n10 = BreakTransform(InTransform=GetDeltaTransformFromRefPose(self=@n3, BoneName=thigh_l, BaseName=pelvis))
   > Set L thigh up(L thigh up:double=MapRangeClamped(Value=BreakRotator(InRot=@n10.Rotation).Yaw, InRangeA=-20, InRangeB=-100)) -> (Output_Get:double)
   n11 = BreakTransform(InTransform=GetDeltaTransformFromRefPose(self=@n3, BoneName=thigh_r, BaseName=pelvis))
   > Set R thigh up(R thigh up:double=MapRangeClamped(Value=BreakRotator(InRot=@n11.Rotation).Yaw, InRangeA=-20, InRangeB=-100)) -> (Output_Get:double)
   n12 = Sequence!(execute=@n2.then_2) -> (then_0, then_1, then_2)
   n13 = Get bIsSkyDiving()
   > n14 = Branch(Condition=@n13)
   > n15 = Set ClothSim_WindStrength(ClothSim_WindStrength:double=0.4) -> (Output_Get:double)
   n16 = Get Mat Instance Dynamic()
   SetScalarParameterValue(
    execute=[@n17.then, @n15.then]
    self=@n16
    ParameterName=ClothSim_WindStrength
    Value=Get ClothSim_WindStrength() -> (ClothSim_WindStrength:double)
   )
   n17 = Set ClothSim_WindStrength(execute=@n14.else, ClothSim_WindStrength=0)
   n18 = Branch(execute=@n12.then_1, Condition=@n13)
   > n19 = Set ClothSim_VelocityMax(ClothSim_VelocityMax:double=750) -> (Output_Get:double)
   SetScalarParameterValue(
    execute=[@n20.then, @n19.then]
    self=@n16
    ParameterName=ClothSim_VelocityMax
    Value=Get ClothSim_VelocityMax() -> (ClothSim_VelocityMax:double)
   )
   n20 = Set ClothSim_VelocityMax(execute=@n18.else, ClothSim_VelocityMax=5000)
   n21 = Branch(execute=@n12.then_2, Condition=@n13)
   > n22 = Set ClothSim_WindNoiseSpeed(ClothSim_WindNoiseSpeed:double=200) -> (Output_Get:double)
   SetScalarParameterValue(
    execute=[@n23.then, @n22.then]
    self=@n16
    ParameterName=ClothSim_WindNoiseSpeed
    Value=Get ClothSim_WindNoiseSpeed() -> (ClothSim_WindNoiseSpeed:double)
   )
   n23 = Set ClothSim_WindNoiseSpeed(execute=@n21.else, ClothSim_WindNoiseSpeed=0)
   Get!(execute:Exec=@n2.then_4) -> (then:Exec, else:Exec, ownerSkelMesh:Skeletal Mesh Component Object Reference)
   > Get!(execute) -> (then, else, PlayerPawn:Fort Player Pawn Athena Object Reference)
   > n24 = PlayerMovement(AnimInstance=Self-Reference(), XDotProductCoefficient=1, YDotProductCoefficient=1, ZDotProductCoefficient=1)
   > n25 = Branch(Condition="Not Equal (Float)"(A=float + float!(A=@n24, B=@n24.YMovement, C:float=@n24.ZMovement) -> (ReturnValue)))
   > Set movementSum(movementSum:bool=true) -> (Output_Get:bool)
   Set movementSum(execute=@n25.else, movementSum=false)
   n26 = Branch(execute=@n2.then_5, Condition=Get movementSum() -> (movementSum:bool))
   SetVectorParameterValue(
    execute=@n26.else
    self=Get Mat Instance Dynamic()
    ParameterName=ClothSim_ABP_R1_Velocity
    Value=Get Value() -> (Value:Linear Color Structure)
   )
   Set WPO Globals Function(
    execute=@n2.then_3
    Material=Get Mat Instance Dynamic()
    Overlay Material=Get Overlay Mat Instance Dynamic()
    Skeletal Mesh=Get skeletal_mesh() -> (skeletal_mesh:Skeletal Mesh Component Object Reference)
    "Local Noise Transform  Location"=Get Local Noise Transform Location() -> (Local Noise Transform Location:Vector)
    Local Noise Offset=Get Local Noise Offset()
   )
   > Set WPO Regional Function!(
    execute
    Material:Material Instance Dynamic Object Reference=Get Mat Instance Dynamic()
    Overlay Material:Material Instance Dynamic Object Reference=Get Overlay Mat Instance Dynamic()
    Skeletal Mesh:Skeletal Mesh Component Object Reference=Get skeletal_mesh()
    Velocity Transform:Name=pelvis
    Velocity Interp Speed:double=Get Velocity Interp Speed()
    Collision Transform:Name=spine_01
    Collision Transform Axis:EAxis Enum
    Invert Collision Transform Axis:bool=false
    Delta Time X:double=@n1.DeltaTimeX
    Previous Velocity Transform Location:"Vector (by ref)"=Get R1 Velocity Transform Location()
    Previous Velocity Transform Velocity:"Vector (by ref)"=Get R1 Velocity Transform Velocity()
    Region:int=1
   ) -> (then)
   n27 = Gravity Override State Function DEPRECATED(
    execute=@n2.then_1
    CharacterPartAnimInstance=Self-Reference()
    GameTimeInSeconds=GetGameTimeInSeconds()
    GravityOverrideParams=Get Gravity Override Params()
   )
   > n28 = Set GravityOverride(GravityOverride=@n27)
   > Set GravityOverride(GravityOverride=SelectVector(A="0,0,-800", B=@n28, bPickA=Get isJumpFalling() -> (isJumpFalling:bool)))
   Branch(execute=@n2.then_6, Condition=Get bIsInFrontEnd())
   > Set customIdleIsPlayingFloat(customIdleIsPlayingFloat:float=GetCurveValue(self=GetOwnerFrontEndAnimInstance(), CurveName=Frontend_Idle_Bypass)) -> (Output_Get:float)
   Event BlueprintInitializeAnimation()
   > n29 = Set ownerSkelMesh(ownerSkelMesh:Skeletal Mesh Component Object Reference=GetOwnerSkeletalMeshComponent()) -> (Output_Get:Skeletal Mesh Component Object Reference)
   > Is Valid(InputObject=@n29)
   > n30 = Cast To FortPlayerPawnAthena(Object=TryGetPawnOwner())
   > Set PlayerPawn(PlayerPawn:Fort Player Pawn Athena Object Reference=@n30) -> (Output_Get:Fort Player Pawn Athena Object Reference)
   Event OnCharacterCustomizationLoaded()
   > n31 = Set skeletal_mesh(skeletal_mesh:Skeletal Mesh Component Object Reference=GetPartSkeletalMeshComponent()) -> (Output_Get:Skeletal Mesh Component Object Reference)
   > n32 = Get Dynamic Material Instance(SkeletalMeshComponent=@n31)
   > Set Mat Instance Dynamic(Mat Instance Dynamic=@n32)
   > n33 = GetOrCreateDynamicOverlayMaterial(MeshComponent=@n31, MaterialName=None)
   > Set Overlay Mat Instance Dynamic(Overlay Mat Instance Dynamic=@n33)
   Event OnEndFalling()
   > Set isJumpFalling(isJumpFalling:bool=false) -> (Output_Get:bool)
   Event OnBeginFalling()
   > Set isJumpFalling(isJumpFalling=true)
   Event OnBeginJumping()
   > Set isJumpFalling(isJumpFalling=true)
  ]}
 GetBoneNames {nodes [
   n1 = GetBoneNames!() -> (then, Region:"Integer (by ref)") {is=title K2Node_FunctionEntry}
   n2 = Append(A=ClothSim_ABP_R, B="To String (Integer)"(InInt=@n1))
   > Return Node!(
    execute
    Velocity:Name=String To Name(InString=Append(A=@n2, B=_Velocity))
    BonePosition:Name=String To Name(InString=Append(A=@n2, B=_BonePosition))
    BoneVector:Name=String To Name(InString=Append(A=@n2, B=_BoneVector))
   )
  ]}
 Set Cloth Sim Params {nodes [
   n1 = Set Cloth Sim Params!() -> (
    then
    material:Material Instance Dynamic Object Reference
    overlayMaterial:Material Instance Dynamic Object Reference
    skelMesh:Skeletal Mesh Component Object Reference
    boneName:Name
    interpSpeed:double=2
    collisionBoneName:Name
    collisionBoneAxis:EAxis Enum=None
    invertCollisionBoneAxis:bool=~
    deltaTimeX:double=~
    previousBoneLocation:"Vector (by ref)"
    previousBoneVelocityVector:"Vector (by ref)"
    region:int=1
   ) {is=title K2Node_FunctionEntry}
   > Set lMID(lMID:Material Instance Dynamic Object Reference=@n1) -> (Output_Get:Material Instance Dynamic Object Reference)
   > Set loverlayMID(loverlayMID:Material Instance Dynamic Object Reference=@n1.overlayMaterial) -> (Output_Get:Material Instance Dynamic Object Reference)
   > Set lSkelMesh(lSkelMesh:Skeletal Mesh Component Object Reference=@n1.skelMesh) -> (Output_Get:Skeletal Mesh Component Object Reference)
   > n2 = GetBoneNames!(execute, self, Region:"Integer (by ref)"=@n1.region) -> (then, Velocity:Name=None, BonePosition:Name=None, BoneVector:Name=None) {is=title K2Node_CallFunction}
   > Set lMatVelocityParamName(lMatVelocityParamName:Name=@n2) -> (Output_Get:Name=None)
   > Set lMatBonePositionParamName(lMatBonePositionParamName:Name=@n2.BonePosition) -> (Output_Get:Name=None)
   > Set lMatBoneVectorParamName(lMatBoneVectorParamName:Name=@n2.BoneVector) -> (Output_Get:Name=None)
   > n3 = calculateMotionVectorForSocket(
    skelMesh=Get lSkelMesh() -> (lSkelMesh:Skeletal Mesh Component Object Reference)
    boneName=@n1.boneName
    previousBoneLocation=@n1.previousBoneLocation
   )
   n4 = VInterpTo!(
    Current:Vector=@n1.previousBoneVelocityVector
    Target:Vector=vector / vector(A:Vector=@n3, B=@n1.deltaTimeX) -> (ReturnValue:Vector)
    DeltaTime:float=@n1.deltaTimeX
    InterpSpeed:float=@n1.interpSpeed
   ) -> (ReturnValue:Vector)
   > SetVectorParameterValue(
    self=[
     Get loverlayMID() -> (loverlayMID:Material Instance Dynamic Object Reference)
     Get lMID() -> (lMID:Material Instance Dynamic Object Reference)
    ]
    ParameterName=Get lMatVelocityParamName() -> (lMatVelocityParamName:Name=None)
    Value="To LinearColor (Vector)"(InVec=@n4)
   )
   n5 = GetSocketTransform(self=Get lSkelMesh(), InSocketName=@n1.collisionBoneName, TransformSpace=RTS_World)
   n6 = "Get Unit Axis (Matrix)"!(
    M:"Matrix Structure (by ref)"="To Matrix (Transform)"!(Transform:"Transform (by ref)"=@n5) -> (ReturnValue:Matrix Structure)
    Axis:EAxis Enum=@n1.collisionBoneAxis
   ) -> (ReturnValue:Vector) {is=title K2Node_CallFunction}
   n7 = Select(Option 0:double=1, Option 1:double=-1, Index=@n1.invertCollisionBoneAxis) -> (ReturnValue:double)
   > SetVectorParameterValue(
    self=[Get loverlayMID(), Get lMID()]
    ParameterName=Get lMatBoneVectorParamName() -> (lMatBoneVectorParamName:Name=None)
    Value="To LinearColor (Vector)"(InVec=vector * vector(A=@n6, B=@n7) -> (ReturnValue=~))
   )
   > SetVectorParameterValue(
    self=[Get loverlayMID(), Get lMID()]
    ParameterName=Get lMatBonePositionParamName() -> (lMatBonePositionParamName:Name=None)
    Value="To LinearColor (Vector)"(InVec=BreakTransform(InTransform=@n5))
   )
   > "Set Vector (by ref)"(Target=@n1.previousBoneLocation, Value=@n3.boneLocation)
   > "Set Vector (by ref)"(Target=@n1.previousBoneVelocityVector, Value=@n4)
   > Return Node!(execute)
   Set lMID()
  ]}
 calculateMotionVectorForSocket {nodes [
   n1 = calculateMotionVectorForSocket!() -> (then, skelMesh:Scene Component Object Reference, boneName:Name, previousBoneLocation:Vector=~) {is=title K2Node_FunctionEntry}
   n2 = GetSocketTransform(self=@n1, InSocketName=@n1.boneName, TransformSpace=RTS_World)
   n3 = BreakTransform(InTransform=@n2)
   > Return Node!(execute, boneMotionVector:Vector=vector - vector(A=@n3, B=@n1.previousBoneLocation), boneLocation:Vector=@n3, boneTransform:Transform=@n2)
  ]}
 setWPOMaterialVelocity {nodes [
   n1 = setWPOMaterialVelocity!() -> (
    then
    materialInstance:Material Instance Dynamic Object Reference
    skelMesh:Skeletal Mesh Component Object Reference
    boneName:Name
    offsetBoneName:Name
    offsetBoneAxis:EAxis Enum=None
    deltaTimeX:double=~
    materialVelocityAttribute:Name=Velocity
    previousBoneLocation:Vector=~
    previousBoneVelocityVector:Vector=~
   ) {is=title K2Node_FunctionEntry}
   > n2 = calculateMotionVectorForSocket(skelMesh=@n1.skelMesh, boneName=@n1.boneName, previousBoneLocation=@n1.previousBoneLocation) {is=title K2Node_CallFunction}
   n3 = "Get Unit Axis (Matrix)"(
    M="To Matrix (Transform)"(Transform=GetSocketTransform(self=@n1.skelMesh, InSocketName=@n1.offsetBoneName, TransformSpace=RTS_World))
    Axis=@n1.offsetBoneAxis
   )
   n4 = vector / vector(A=@n2, B=@n1.deltaTimeX)
   n5 = VInterpTo(Current=@n1.previousBoneVelocityVector, Target=@n4, DeltaTime=@n1.deltaTimeX, InterpSpeed=0.75)
   > SetVectorParameterValue(self=@n1, ParameterName=@n1.materialVelocityAttribute, Value="To LinearColor (Vector)"(InVec=vector + vector(A=@n3, B=@n5)))
   > Return Node!(
    execute
    boneVelocity:double=Vector Length(A=@n4)
    boneVelocityVector:Vector=@n4
    boneLocation=@n2.boneLocation
    boneTransform=@n2.boneTransform
   )
  ]}
]
variables [
 armsBack {type=number}
 ClothSim_VelocityMax {type=number}
 ClothSim_WindNoiseSpeed {type=number}
 ClothSim_WindStrength {type=number}
 customIdleIsPlayingFloat {type=number}
 Gravity Override Params {type=GravityOverrideParamsStruct, default {skydive_DivingDown {gravityOverride=<-1500 0 100>}}}
 GravityOverride {type=Vector}
 isJumpFalling {type=boolean}
 L clav alpha {type=number}
 L thigh up {type=number}
 L upperarm alpha {type=number}
 Local Noise Offset {type=Vector}
 Local Noise Transform Location {type=Vector}
 Mat Instance Dynamic {type=MaterialInstanceDynamic}
 movementSum {type=boolean}
 neckRotateBack {type=number}
 Overlay Mat Instance Dynamic {type=MaterialInstanceDynamic}
 ownerSkelMesh {type=SkeletalMeshComponent}
 PlayerPawn {type=FortPlayerPawnAthena}
 R clav alpha {type=number}
 R thigh up {type=number}
 R upperarm alpha {type=number}
 R1 Velocity Transform Location {type=Vector}
 R1 Velocity Transform Velocity {type=Vector}
 skeletal_mesh {type=SkeletalMeshComponent}
 Value {type=LinearColor, default=<rgba 0 0 0 0>}
 Velocity Interp Speed {type=number}
]
